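# Translucent, two-sided lit material for the GrainSpike emote prop, drawing an animated blue portal with glowing ripples.
# Time-driven sine waves inside sphere masks form rings, broken up by panning noise and UV distortion.
# Layered dark-blue masks, a radially mapped, tinted portal texture and panning flake "Bling" supply color and sparkle.
# A saturated EmotesGrainSpike parameter provides the animation value, alongside a blendable offset toward the object origin.
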Material /CosmeticCompanions/Assets/Props/Emotes/GrainSpike/Material/M_GrainSpike_Ripples
graph {
 expressions {
  MaterialExpressionAdd_0 {
   class=MaterialExpressionAdd
   inputs [
    A {from=MaterialExpressionPower_0}
    B {from=MaterialExpressionMultiply_16}
   ]
  }
  MaterialExpressionAdd_1 {
   class=MaterialExpressionAdd
   inputs [
    A {from=MaterialExpressionAdd_9}
    B {from=MaterialExpressionReroute_1}
   ]
  }
  MaterialExpressionAdd_11 {
   class=MaterialExpressionAdd
   inputs [
    A {from=MaterialExpressionMultiply_34}
    B {from=MaterialExpressionTextureCoordinate_1}
   ]
  }
  MaterialExpressionAdd_2 {
   class=MaterialExpressionAdd
   inputs [
    A {from=MaterialExpressionPower_2}
    B {from=MaterialExpressionMultiply_4}
   ]
  }
  MaterialExpressionAdd_3 {
   class=MaterialExpressionAdd
   inputs [
    A {from=MaterialExpressionMultiply_7}
    B {from=MaterialExpressionTextureCoordinate_3}
   ]
  }
  MaterialExpressionAdd_4 {
   class=MaterialExpressionAdd
   inputs [
    A {from=MaterialExpressionReroute_7}
    B {from=MaterialExpressionSaturate_3}
   ]
  }
  MaterialExpressionAdd_6 {
   class=MaterialExpressionAdd
   inputs [
    A {from=MaterialExpressionPower_4}
    B {from=MaterialExpressionMultiply_5}
   ]
  }
  MaterialExpressionAdd_7 {
   class=MaterialExpressionAdd
   inputs [
    A {from=MaterialExpressionMultiply_13}
    B {from=MaterialExpressionAdd_6}
   ]
  }
  MaterialExpressionAdd_8 {
   class=MaterialExpressionAdd
   inputs [
    A {from=MaterialExpressionPower_9}
    B {from=MaterialExpressionMultiply_2}
   ]
  }
  MaterialExpressionAdd_9 {
   class=MaterialExpressionAdd
   inputs [
    A {from=MaterialExpressionMultiply_27}
    B {from=MaterialExpressionMultiply_19}
   ]
  }
  MaterialExpressionAppendVector_2 {
   class=MaterialExpressionAppendVector
   inputs [
    A {from=MaterialExpressionConstant_39, output=""}
    B {from=MaterialExpressionConstant_41, output=""}
   ]
  }
  MaterialExpressionAppendVector_3 {
   class=MaterialExpressionAppendVector
   inputs [
    A {from=MaterialExpressionConstant_40, output=""}
    B {from=MaterialExpressionConstant_42, output=""}
   ]
  }
  MaterialExpressionComponentMask_0 {class=MaterialExpressionComponentMask, inputs [None {from=MaterialExpressionTextureSample_0, output=RGB}], properties {g=true, r=true}}
  MaterialExpressionComponentMask_2 {class=MaterialExpressionComponentMask, inputs [None {from=MaterialExpressionTextureSample_4, output=RGB}], properties {g=true, r=true}}
  MaterialExpressionConstant3Vector_0 {class=MaterialExpressionConstant3Vector}
  MaterialExpressionConstant4Vector_2 {class=MaterialExpressionConstant4Vector, inputs [], properties {constant=<rgba 0.005 0.0186 0.05 1>}}
  MaterialExpressionConstant4Vector_3 {class=MaterialExpressionConstant4Vector, inputs [], properties {constant=<rgba 0.006 0.0224 0.06 1>}}
  MaterialExpressionConstant4Vector_4 {class=MaterialExpressionConstant4Vector, inputs [], properties {constant=<rgba 0.0008 0.0083 0.02 1>}}
  MaterialExpressionConstant_0 {class=MaterialExpressionConstant, properties {r=6}}
  MaterialExpressionConstant_10 {class=MaterialExpressionConstant, properties {r=0.16}}
  MaterialExpressionConstant_11 {class=MaterialExpressionConstant, properties {r=0.62}}
  MaterialExpressionConstant_12 {class=MaterialExpressionConstant, properties {r=0.226}}
  MaterialExpressionConstant_13 {class=MaterialExpressionConstant, properties {r=0.572}}
  MaterialExpressionConstant_14 {class=MaterialExpressionConstant, properties {r=9.4866}}
  MaterialExpressionConstant_17 {class=MaterialExpressionConstant, properties {r=0.138}}
  MaterialExpressionConstant_19 {class=MaterialExpressionConstant, properties {r=0.05}}
  MaterialExpressionConstant_20 {class=MaterialExpressionConstant, properties {r=1.3}}
  MaterialExpressionConstant_21 {class=MaterialExpressionConstant, properties {r=1.2063}}
  MaterialExpressionConstant_22 {class=MaterialExpressionConstant, properties {r=6}}
  MaterialExpressionConstant_23 {class=MaterialExpressionConstant, properties {r=0.6444}}
  MaterialExpressionConstant_24 {class=MaterialExpressionConstant, properties {r=0.1233}}
  MaterialExpressionConstant_25 {class=MaterialExpressionConstant, properties {r=1.9396}}
  MaterialExpressionConstant_26 {class=MaterialExpressionConstant, properties {r=-10}}
  MaterialExpressionConstant_28 {class=MaterialExpressionConstant}
  MaterialExpressionConstant_29 {class=MaterialExpressionConstant, properties {r=2}}
  MaterialExpressionConstant_3 {class=MaterialExpressionConstant, properties {r=6}}
  MaterialExpressionConstant_30 {class=MaterialExpressionConstant, properties {r=1.7295}}
  MaterialExpressionConstant_31 {class=MaterialExpressionConstant, properties {r=4.468}}
  MaterialExpressionConstant_32 {class=MaterialExpressionConstant, properties {r=10}}
  MaterialExpressionConstant_33 {class=MaterialExpressionConstant, properties {r=0.1}}
  MaterialExpressionConstant_34 {class=MaterialExpressionConstant, properties {r=0.5}}
  MaterialExpressionConstant_35 {class=MaterialExpressionConstant, properties {r=0.5}}
  MaterialExpressionConstant_36 {class=MaterialExpressionConstant, properties {r=0.15}}
  MaterialExpressionConstant_37 {class=MaterialExpressionConstant, properties {r=2.148}}
  MaterialExpressionConstant_38 {class=MaterialExpressionConstant, properties {r=109.8134}}
  MaterialExpressionConstant_39 {class=MaterialExpressionConstant, properties {r=10}}
  MaterialExpressionConstant_40 {class=MaterialExpressionConstant, properties {r=0.1}}
  MaterialExpressionConstant_41 {class=MaterialExpressionConstant, properties {r=0.3}}
  MaterialExpressionConstant_42 {class=MaterialExpressionConstant, properties {r=-0.002}}
  MaterialExpressionConstant_43 {class=MaterialExpressionConstant, properties {r=0.9}}
  MaterialExpressionConstant_44 {class=MaterialExpressionConstant, properties {r=19.7165}}
  MaterialExpressionConstant_46 {class=MaterialExpressionConstant, properties {r=2}}
  MaterialExpressionConstant_47 {class=MaterialExpressionConstant, properties {r=0.2}}
  MaterialExpressionConstant_48 {class=MaterialExpressionConstant, properties {r=0.92}}
  MaterialExpressionConstant_49 {class=MaterialExpressionConstant, properties {r=1.2}}
  MaterialExpressionConstant_50 {class=MaterialExpressionConstant, properties {r=0.5}}
  MaterialExpressionConstant_6 {class=MaterialExpressionConstant, properties {r=4}}
  MaterialExpressionConstant_7 {class=MaterialExpressionConstant, properties {r=9.4867}}
  MaterialExpressionConstant_9 {class=MaterialExpressionConstant, properties {r=1.2}}
  MaterialExpressionConvert_0 {
   class=MaterialExpressionConvert
   inputs [
    None {from=MaterialExpressionConstant_35}
    None {from=MaterialExpressionConstant_34}
   ]
   properties {
    nodeName=Make Vector2
    convertInputs [
     {
      type=Scalar
      expressionInput {
       expression=&self:MaterialExpressionConstant_35
       inputName=None
       mask=0
       maskA=0
       maskB=0
       maskG=0
       maskR=0
       outputIndex=0
      }
     }
     {
      type=Scalar
      expressionInput {
       expression=&self:MaterialExpressionConstant_34
       inputName=None
       mask=0
       maskA=0
       maskB=0
       maskG=0
       maskR=0
       outputIndex=0
      }
     }
    ]
    convertOutputs [{type=Vector2}]
   }
  }
  MaterialExpressionDDX_0 {class=MaterialExpressionDDX, inputs [Value {from=MaterialExpressionTextureCoordinate_5, output=""}], properties {}}
  MaterialExpressionDDY_0 {class=MaterialExpressionDDY, inputs [Value {from=MaterialExpressionTextureCoordinate_5, output=""}], properties {}}
  MaterialExpressionDesaturation_0 {
   class=MaterialExpressionDesaturation
   inputs [
    None {from=MaterialExpressionTextureSample_2, output=RGB}
    Fraction {from=MaterialExpressionLinearInterpolate_7, output=""}
   ]
   properties {}
  }
  MaterialExpressionLinearInterpolate_0 {
   class=MaterialExpressionLinearInterpolate
   inputs [
    A {from=MaterialExpressionSubtract_3}
    B {from=""}
    Alpha {from=MaterialExpressionNamedRerouteUsage_38}
   ]
   properties {constB=0}
  }
  MaterialExpressionLinearInterpolate_1 {
   class=MaterialExpressionLinearInterpolate
   inputs [
    A {from=MaterialExpressionVectorParameter_0, output=RGB}
    B {from=MaterialExpressionTextureSample_2, output=RGB}
    Alpha {from=""}
   ]
   properties {constAlpha=0.05}
  }
  MaterialExpressionLinearInterpolate_2 {
   class=MaterialExpressionLinearInterpolate
   inputs [
    A {from=""}
    B {from=MaterialExpressionNamedRerouteUsage_39}
    Alpha {from=MaterialExpressionNamedRerouteUsage_35}
   ]
   properties {constA=1}
  }
  MaterialExpressionLinearInterpolate_3 {
   class=MaterialExpressionLinearInterpolate
   inputs [
    A {from=MaterialExpressionVectorParameter_0, output=RGB}
    B {from=MaterialExpressionDesaturation_0}
    Alpha {from=MaterialExpressionSaturate_6}
   ]
  }
  MaterialExpressionLinearInterpolate_4 {
   class=MaterialExpressionLinearInterpolate
   inputs [
    A {from=MaterialExpressionScalarParameter_5}
    B {from=MaterialExpressionConstant_26}
    Alpha {from=MaterialExpressionPower_5}
   ]
  }
  MaterialExpressionLinearInterpolate_5 {
   class=MaterialExpressionLinearInterpolate
   inputs [
    A {from=MaterialExpressionReroute_6}
    B {from=MaterialExpressionLinearInterpolate_1}
    Alpha {from=MaterialExpressionNamedRerouteUsage_33}
   ]
  }
  MaterialExpressionLinearInterpolate_6 {
   class=MaterialExpressionLinearInterpolate
   inputs [
    A {from=MaterialExpressionConstant4Vector_3, output=RGB}
    B {from=MaterialExpressionConstant4Vector_4, output=RGB}
    Alpha {from=MaterialExpressionSphereMask_8}
   ]
  }
  MaterialExpressionLinearInterpolate_7 {
   class=MaterialExpressionLinearInterpolate
   inputs [
    A {from=""}
    B {from=MaterialExpressionScalarParameter_11}
    Alpha {from=MaterialExpressionMultiply_30}
   ]
   properties {constB=-1}
  }
  MaterialExpressionLinearInterpolate_9 {
   class=MaterialExpressionLinearInterpolate
   inputs [
    A {from=MaterialExpressionLinearInterpolate_6}
    B {from=MaterialExpressionConstant4Vector_2, output=RGB}
    Alpha {from=MaterialExpressionPower_8}
   ]
  }
  MaterialExpressionMaterialFunctionCall_0 {
   class=MaterialExpressionMaterialFunctionCall
   inputs [
    Vector or UVs {from=""}
    Swizzle Coordinate Output {from=""}
   ]
   properties {materialFunction=&VectorToRadialValue}
  }
  MaterialExpressionMaterialFunctionCall_1 {
   class=MaterialExpressionMaterialFunctionCall
   inputs [
    Vector or UVs {from=""}
    Swizzle Coordinate Output {from=""}
   ]
   properties {materialFunction=&VectorToRadialValue}
  }
  MaterialExpressionMax_0 {
   class=MaterialExpressionMax
   inputs [
    A {from=MaterialExpressionSaturate_2}
    B {from=MaterialExpressionMultiply_33}
   ]
  }
  MaterialExpressionMultiply_0 {
   class=MaterialExpressionMultiply
   inputs [
    A {from=MaterialExpressionSaturate_4}
    B {from=MaterialExpressionNamedRerouteUsage_37}
   ]
  }
  MaterialExpressionMultiply_1 {
   class=MaterialExpressionMultiply
   inputs [
    A {from=MaterialExpressionTextureCoordinate_2}
    B {from=MaterialExpressionConstant_20}
   ]
  }
  MaterialExpressionMultiply_10 {
   class=MaterialExpressionMultiply
   inputs [
    A {from=MaterialExpressionScalarParameter_4}
    B {from=""}
   ]
   properties {constB=0.5}
  }
  MaterialExpressionMultiply_11 {
   class=MaterialExpressionMultiply
   inputs [
    A {from=MaterialExpressionSphereMask_0}
    B {from=MaterialExpressionMax_0}
   ]
  }
  MaterialExpressionMultiply_12 {
   class=MaterialExpressionMultiply
   inputs [
    A {from=MaterialExpressionLinearInterpolate_2}
    B {from=MaterialExpressionMultiply_14}
   ]
  }
  MaterialExpressionMultiply_13 {
   class=MaterialExpressionMultiply
   inputs [
    A {from=MaterialExpressionNamedRerouteUsage_36}
    B {from=MaterialExpressionConstant_43}
   ]
  }
  MaterialExpressionMultiply_14 {
   class=MaterialExpressionMultiply
   inputs [
    A {from=MaterialExpressionLinearInterpolate_5}
    B {from=MaterialExpressionScalarParameter_6}
   ]
  }
  MaterialExpressionMultiply_15 {
   class=MaterialExpressionMultiply
   inputs [
    A {from=MaterialExpressionMultiply_3}
    B {from=MaterialExpressionConstant_6}
   ]
  }
  MaterialExpressionMultiply_16 {
   class=MaterialExpressionMultiply
   inputs [
    A {from=MaterialExpressionConstant_33}
    B {from=MaterialExpressionTime_0}
   ]
  }
  MaterialExpressionMultiply_17 {
   class=MaterialExpressionMultiply
   inputs [
    A {from=MaterialExpressionAdd_0}
    B {from=MaterialExpressionConstant_31}
   ]
   properties {constB=5}
  }
  MaterialExpressionMultiply_18 {
   class=MaterialExpressionMultiply
   inputs [
    A {from=MaterialExpressionSine_1}
    B {from=MaterialExpressionConstant_32}
   ]
  }
  MaterialExpressionMultiply_19 {
   class=MaterialExpressionMultiply
   inputs [
    A {from=MaterialExpressionNamedRerouteUsage_20}
    B {from=MaterialExpressionConstant_50}
   ]
  }
  MaterialExpressionMultiply_2 {
   class=MaterialExpressionMultiply
   inputs [
    A {from=MaterialExpressionConstant_36}
    B {from=MaterialExpressionTime_2}
   ]
  }
  MaterialExpressionMultiply_20 {
   class=MaterialExpressionMultiply
   inputs [
    A {from=MaterialExpressionNamedRerouteUsage_25}
    B {from=MaterialExpressionMultiply_9}
   ]
  }
  MaterialExpressionMultiply_21 {
   class=MaterialExpressionMultiply
   inputs [
    A {from=MaterialExpressionNamedRerouteDeclaration_9}
    B {from=MaterialExpressionConstant_29}
   ]
   properties {constB=2}
  }
  MaterialExpressionMultiply_22 {
   class=MaterialExpressionMultiply
   inputs [
    A {from=MaterialExpressionTextureCoordinate_4}
    B {from=MaterialExpressionConstant_9}
   ]
  }
  MaterialExpressionMultiply_23 {
   class=MaterialExpressionMultiply
   inputs [
    A {from=MaterialExpressionMaterialFunctionCall_0, output=Radial Coordinates}
    B {from=MaterialExpressionConstant_47}
   ]
   properties {constB=0.5}
  }
  MaterialExpressionMultiply_24 {
   class=MaterialExpressionMultiply
   inputs [
    A {from=MaterialExpressionMaterialFunctionCall_1, output=Radial Coordinates}
    B {from=MaterialExpressionAppendVector_2}
   ]
  }
  MaterialExpressionMultiply_25 {
   class=MaterialExpressionMultiply
   inputs [
    A {from=MaterialExpressionNamedRerouteUsage_16}
    B {from=MaterialExpressionConstant_49}
   ]
   properties {constB=3}
  }
  MaterialExpressionMultiply_26 {
   class=MaterialExpressionMultiply
   inputs [
    A {from=MaterialExpressionAdd_8}
    B {from=MaterialExpressionConstant_37}
   ]
   properties {constB=5}
  }
  MaterialExpressionMultiply_27 {
   class=MaterialExpressionMultiply
   inputs [
    A {from=MaterialExpressionSubtract_6}
    B {from=MaterialExpressionConstant_14}
   ]
  }
  MaterialExpressionMultiply_28 {
   class=MaterialExpressionMultiply
   inputs [
    A {from=MaterialExpressionNamedRerouteDeclaration_10}
    B {from=MaterialExpressionMultiply_11}
   ]
  }
  MaterialExpressionMultiply_29 {
   class=MaterialExpressionMultiply
   inputs [
    A {from=MaterialExpressionNamedRerouteUsage_27}
    B {from=MaterialExpressionConstant_21}
   ]
   properties {constB=2}
  }
  MaterialExpressionMultiply_3 {
   class=MaterialExpressionMultiply
   inputs [
    A {from=MaterialExpressionSphereMask_2}
    B {from=MaterialExpressionSaturate_0}
   ]
  }
  MaterialExpressionMultiply_30 {
   class=MaterialExpressionMultiply
   inputs [
    A {from=MaterialExpressionNamedRerouteUsage_26}
    B {from=MaterialExpressionPower_1}
   ]
  }
  MaterialExpressionMultiply_32 {
   class=MaterialExpressionMultiply
   inputs [
    A {from=MaterialExpressionSine_2}
    B {from=MaterialExpressionConstant_38}
   ]
  }
  MaterialExpressionMultiply_33 {
   class=MaterialExpressionMultiply
   inputs [
    A {from=MaterialExpressionSaturate_5}
    B {from=""}
   ]
   properties {constB=0.5}
  }
  MaterialExpressionMultiply_34 {
   class=MaterialExpressionMultiply
   inputs [
    A {from=MaterialExpressionComponentMask_0}
    B {from=MaterialExpressionConstant_19}
   ]
  }
  MaterialExpressionMultiply_35 {
   class=MaterialExpressionMultiply
   inputs [
    A {from=MaterialExpressionConstant_19}
    B {from=""}
   ]
   properties {constB=0.5}
  }
  MaterialExpressionMultiply_36 {
   class=MaterialExpressionMultiply
   inputs [
    A {from=MaterialExpressionTextureSample_3, output=R}
    B {from=MaterialExpressionConstant_46}
   ]
   properties {constB=2}
  }
  MaterialExpressionMultiply_4 {
   class=MaterialExpressionMultiply
   inputs [
    A {from=MaterialExpressionScalarParameter_2}
    B {from=MaterialExpressionTime_1}
   ]
  }
  MaterialExpressionMultiply_5 {
   class=MaterialExpressionMultiply
   inputs [
    A {from=MaterialExpressionMultiply_25}
    B {from=MaterialExpressionPower_3}
   ]
  }
  MaterialExpressionMultiply_6 {
   class=MaterialExpressionMultiply
   inputs [
    A {from=MaterialExpressionAdd_2}
    B {from=MaterialExpressionScalarParameter_3}
   ]
   properties {constB=5}
  }
  MaterialExpressionMultiply_7 {
   class=MaterialExpressionMultiply
   inputs [
    A {from=MaterialExpressionComponentMask_0}
    B {from=MaterialExpressionScalarParameter_4}
   ]
  }
  MaterialExpressionMultiply_8 {
   class=MaterialExpressionMultiply
   inputs [
    A {from=MaterialExpressionSine_0}
    B {from=MaterialExpressionLinearInterpolate_4}
   ]
  }
  MaterialExpressionMultiply_9 {
   class=MaterialExpressionMultiply
   inputs [
    A {from=MaterialExpressionSphereMask_3}
    B {from=MaterialExpressionMultiply_15}
   ]
  }
  MaterialExpressionNamedRerouteDeclaration_0 {
   class=MaterialExpressionNamedRerouteDeclaration
   inputs [None {from=MaterialExpressionSubtract_2}]
   properties {name=PortalUV, nodeColor=<rgba 1 0.7969 0 1>}
  }
  MaterialExpressionNamedRerouteDeclaration_1 {
   class=MaterialExpressionNamedRerouteDeclaration
   inputs [None {from=MaterialExpressionConvert_0}]
   properties {name=UV center position, nodeColor=<rgba 0.1719 1 0 1>}
  }
  MaterialExpressionNamedRerouteDeclaration_10 {
   class=MaterialExpressionNamedRerouteDeclaration
   inputs [None {from=MaterialExpressionPower_6}]
   properties {name=NoiseB, nodeColor=<rgba 0 0.1562 1 1>}
  }
  MaterialExpressionNamedRerouteDeclaration_11 {
   class=MaterialExpressionNamedRerouteDeclaration
   inputs [None {from=MaterialExpressionSphereMask_5}]
   properties {name=PortalMask, nodeColor=<rgba 0.4766 1 0 1>}
  }
  MaterialExpressionNamedRerouteDeclaration_12 {
   class=MaterialExpressionNamedRerouteDeclaration
   inputs [None {from=MaterialExpressionSaturate_1}]
   properties {name=EmoteAnimation, nodeColor=<rgba 1 0 0.7969 1>}
  }
  MaterialExpressionNamedRerouteDeclaration_13 {
   class=MaterialExpressionNamedRerouteDeclaration
   inputs [None {from=MaterialExpressionReroute_0}]
   properties {name=PortalColor, nodeColor=<rgba 0 1 0.8828 1>}
  }
  MaterialExpressionNamedRerouteDeclaration_2 {
   class=MaterialExpressionNamedRerouteDeclaration
   inputs [None {from=MaterialExpressionScalarParameter_14}]
   properties {name=Ripples_Radius, nodeColor=<rgba 1 0 0.3984 1>}
  }
  MaterialExpressionNamedRerouteDeclaration_3 {
   class=MaterialExpressionNamedRerouteDeclaration
   inputs [None {from=MaterialExpressionSphereMask_1}]
   properties {name=RipplesMask, nodeColor=<rgba 0 0.7187 1 1>}
  }
  MaterialExpressionNamedRerouteDeclaration_4 {
   class=MaterialExpressionNamedRerouteDeclaration
   inputs [None {from=MaterialExpressionMultiply_28}]
   properties {name=InterRipple, nodeColor=<rgba 0.5703 1 0 1>}
  }
  MaterialExpressionNamedRerouteDeclaration_5 {
   class=MaterialExpressionNamedRerouteDeclaration
   inputs [None {from=MaterialExpressionAdd_4}]
   properties {name=Bling, nodeColor=<rgba 0.5234 1 0 1>}
  }
  MaterialExpressionNamedRerouteDeclaration_6 {
   class=MaterialExpressionNamedRerouteDeclaration
   inputs [None {from=MaterialExpressionSubtract_9}]
   properties {name=PortalUV_1, nodeColor=<rgba 1 0.7969 0 1>}
  }
  MaterialExpressionNamedRerouteDeclaration_7 {
   class=MaterialExpressionNamedRerouteDeclaration
   inputs [None {from=MaterialExpressionSphereMask_9}]
   properties {name=RipplesMask_Variation, nodeColor=<rgba 0 0.7188 1 1>}
  }
  MaterialExpressionNamedRerouteDeclaration_8 {
   class=MaterialExpressionNamedRerouteDeclaration
   inputs [None {from=MaterialExpressionMultiply_20}]
   properties {name=Ripples, nodeColor=<rgba 0 1 0.4844 1>}
  }
  MaterialExpressionNamedRerouteDeclaration_9 {
   class=MaterialExpressionNamedRerouteDeclaration
   inputs [None {from=MaterialExpressionTextureSample_1, output=R}]
   properties {name=NoiseA, nodeColor=<rgba 1 0.1641 0 1>}
  }
  MaterialExpressionNamedRerouteUsage_0 {class=MaterialExpressionNamedRerouteUsage}
  MaterialExpressionNamedRerouteUsage_1 {class=MaterialExpressionNamedRerouteUsage}
  MaterialExpressionNamedRerouteUsage_10 {class=MaterialExpressionNamedRerouteUsage}
  MaterialExpressionNamedRerouteUsage_11 {class=MaterialExpressionNamedRerouteUsage}
  MaterialExpressionNamedRerouteUsage_12 {class=MaterialExpressionNamedRerouteUsage}
  MaterialExpressionNamedRerouteUsage_13 {class=MaterialExpressionNamedRerouteUsage}
  MaterialExpressionNamedRerouteUsage_14 {class=MaterialExpressionNamedRerouteUsage}
  MaterialExpressionNamedRerouteUsage_15 {class=MaterialExpressionNamedRerouteUsage}
  MaterialExpressionNamedRerouteUsage_16 {class=MaterialExpressionNamedRerouteUsage}
  MaterialExpressionNamedRerouteUsage_17 {class=MaterialExpressionNamedRerouteUsage}
  MaterialExpressionNamedRerouteUsage_18 {class=MaterialExpressionNamedRerouteUsage}
  MaterialExpressionNamedRerouteUsage_19 {class=MaterialExpressionNamedRerouteUsage}
  MaterialExpressionNamedRerouteUsage_2 {class=MaterialExpressionNamedRerouteUsage}
  MaterialExpressionNamedRerouteUsage_20 {class=MaterialExpressionNamedRerouteUsage}
  MaterialExpressionNamedRerouteUsage_21 {class=MaterialExpressionNamedRerouteUsage}
  MaterialExpressionNamedRerouteUsage_22 {class=MaterialExpressionNamedRerouteUsage}
  MaterialExpressionNamedRerouteUsage_23 {class=MaterialExpressionNamedRerouteUsage}
  MaterialExpressionNamedRerouteUsage_24 {class=MaterialExpressionNamedRerouteUsage}
  MaterialExpressionNamedRerouteUsage_25 {class=MaterialExpressionNamedRerouteUsage}
  MaterialExpressionNamedRerouteUsage_26 {class=MaterialExpressionNamedRerouteUsage}
  MaterialExpressionNamedRerouteUsage_27 {class=MaterialExpressionNamedRerouteUsage}
  MaterialExpressionNamedRerouteUsage_28 {class=MaterialExpressionNamedRerouteUsage}
  MaterialExpressionNamedRerouteUsage_29 {class=MaterialExpressionNamedRerouteUsage}
  MaterialExpressionNamedRerouteUsage_3 {class=MaterialExpressionNamedRerouteUsage}
  MaterialExpressionNamedRerouteUsage_30 {class=MaterialExpressionNamedRerouteUsage}
  MaterialExpressionNamedRerouteUsage_31 {class=MaterialExpressionNamedRerouteUsage}
  MaterialExpressionNamedRerouteUsage_32 {class=MaterialExpressionNamedRerouteUsage}
  MaterialExpressionNamedRerouteUsage_33 {class=MaterialExpressionNamedRerouteUsage}
  MaterialExpressionNamedRerouteUsage_34 {class=MaterialExpressionNamedRerouteUsage}
  MaterialExpressionNamedRerouteUsage_35 {class=MaterialExpressionNamedRerouteUsage}
  MaterialExpressionNamedRerouteUsage_36 {class=MaterialExpressionNamedRerouteUsage}
  MaterialExpressionNamedRerouteUsage_37 {class=MaterialExpressionNamedRerouteUsage}
  MaterialExpressionNamedRerouteUsage_38 {class=MaterialExpressionNamedRerouteUsage}
  MaterialExpressionNamedRerouteUsage_39 {class=MaterialExpressionNamedRerouteUsage}
  MaterialExpressionNamedRerouteUsage_4 {class=MaterialExpressionNamedRerouteUsage}
  MaterialExpressionNamedRerouteUsage_5 {class=MaterialExpressionNamedRerouteUsage}
  MaterialExpressionNamedRerouteUsage_6 {class=MaterialExpressionNamedRerouteUsage}
  MaterialExpressionNamedRerouteUsage_7 {class=MaterialExpressionNamedRerouteUsage}
  MaterialExpressionNamedRerouteUsage_8 {class=MaterialExpressionNamedRerouteUsage}
  MaterialExpressionNamedRerouteUsage_9 {class=MaterialExpressionNamedRerouteUsage}
  MaterialExpressionPanner_0 {
   class=MaterialExpressionPanner
   inputs [
    Coordinate {from=MaterialExpressionMultiply_22}
    Time {from=""}
    Speed {from=""}
   ]
   properties {speedX=0.1}
  }
  MaterialExpressionPanner_1 {
   class=MaterialExpressionPanner
   inputs [
    Coordinate {from=MaterialExpressionMultiply_1}
    Time {from=""}
    Speed {from=""}
   ]
   properties {speedX=0.05, speedY=-0.02}
  }
  MaterialExpressionPanner_3 {
   class=MaterialExpressionPanner
   inputs [
    Coordinate {from=MaterialExpressionMultiply_24}
    Time {from=""}
    Speed {from=MaterialExpressionAppendVector_3}
   ]
   properties {speedX=0.03, speedY=-0.002}
  }
  MaterialExpressionPower_0 {
   class=MaterialExpressionPower
   inputs [
    Base {from=MaterialExpressionNamedRerouteUsage_19}
    Exp {from=MaterialExpressionConstant_3}
   ]
  }
  MaterialExpressionPower_1 {
   class=MaterialExpressionPower
   inputs [
    Base {from=MaterialExpressionNamedRerouteUsage_14}
    Exp {from=""}
   ]
   properties {constExponent=8}
  }
  MaterialExpressionPower_2 {
   class=MaterialExpressionPower
   inputs [
    Base {from=MaterialExpressionNamedRerouteUsage_13}
    Exp {from=MaterialExpressionConstant_22}
   ]
  }
  MaterialExpressionPower_3 {
   class=MaterialExpressionPower
   inputs [
    Base {from=MaterialExpressionNamedRerouteUsage_11}
    Exp {from=""}
   ]
   properties {constExponent=10}
  }
  MaterialExpressionPower_4 {
   class=MaterialExpressionPower
   inputs [
    Base {from=MaterialExpressionNamedRerouteUsage_12}
    Exp {from=MaterialExpressionConstant_44}
   ]
   properties {constExponent=13.5}
  }
  MaterialExpressionPower_5 {
   class=MaterialExpressionPower
   inputs [
    Base {from=MaterialExpressionSphereMask_6}
    Exp {from=MaterialExpressionConstant_25}
   ]
  }
  MaterialExpressionPower_6 {
   class=MaterialExpressionPower
   inputs [
    Base {from=MaterialExpressionMultiply_21}
    Exp {from=MaterialExpressionConstant_30}
   ]
  }
  MaterialExpressionPower_8 {
   class=MaterialExpressionPower
   inputs [
    Base {from=MaterialExpressionNamedRerouteUsage_28}
    Exp {from=MaterialExpressionConstant_7}
   ]
   properties {constExponent=10}
  }
  MaterialExpressionPower_9 {
   class=MaterialExpressionPower
   inputs [
    Base {from=MaterialExpressionNamedRerouteUsage_32}
    Exp {from=MaterialExpressionConstant_0}
   ]
  }
  MaterialExpressionReroute_0 {class=MaterialExpressionReroute, inputs [None {from=MaterialExpressionLinearInterpolate_9}]}
  MaterialExpressionReroute_1 {class=MaterialExpressionReroute, inputs [None {from=MaterialExpressionMultiply_12}]}
  MaterialExpressionReroute_2 {class=MaterialExpressionReroute, inputs [None {from=MaterialExpressionAdd_1}]}
  MaterialExpressionReroute_3 {class=MaterialExpressionReroute, inputs [None {from=MaterialExpressionMultiply_0}]}
  MaterialExpressionReroute_6 {class=MaterialExpressionReroute, inputs [None {from=MaterialExpressionLinearInterpolate_3}]}
  MaterialExpressionReroute_7 {class=MaterialExpressionReroute, inputs [None {from=MaterialExpressionMultiply_36}]}
  MaterialExpressionSaturate_0 {class=MaterialExpressionSaturate, inputs [None {from=MaterialExpressionSubtract_1}]}
  MaterialExpressionSaturate_1 {class=MaterialExpressionSaturate, inputs [None {from=MaterialExpressionScalarParameter_0}]}
  MaterialExpressionSaturate_2 {class=MaterialExpressionSaturate, inputs [None {from=MaterialExpressionSubtract_5}]}
  MaterialExpressionSaturate_3 {class=MaterialExpressionSaturate, inputs [None {from=MaterialExpressionNamedRerouteUsage_29}]}
  MaterialExpressionSaturate_4 {class=MaterialExpressionSaturate, inputs [None {from=MaterialExpressionAdd_7}]}
  MaterialExpressionSaturate_5 {class=MaterialExpressionSaturate, inputs [None {from=MaterialExpressionSubtract_8}]}
  MaterialExpressionSaturate_6 {class=MaterialExpressionSaturate, inputs [None {from=MaterialExpressionMultiply_29}]}
  MaterialExpressionScalarParameter_0 {class=MaterialExpressionScalarParameter, properties {defaultValue=1, group=Ripples, parameterName=EmotesGrainSpike}}
  MaterialExpressionScalarParameter_11 {class=MaterialExpressionScalarParameter, properties {defaultValue=-8, group=Ripples, parameterName=ColorDesaturation}}
  MaterialExpressionScalarParameter_14 {class=MaterialExpressionScalarParameter, properties {defaultValue=0.85, group=Ripples, parameterName=Ripples_Radius}}
  MaterialExpressionScalarParameter_2 {class=MaterialExpressionScalarParameter, properties {defaultValue=0.12, group=Ripples, parameterName=Ripples_Speed}}
  MaterialExpressionScalarParameter_3 {class=MaterialExpressionScalarParameter, properties {defaultValue=16, group=Ripples, parameterName=Ripples_Tiling}}
  MaterialExpressionScalarParameter_4 {class=MaterialExpressionScalarParameter, properties {defaultValue=0.03, group=Ripples, parameterName=Ripples_DistortAmount}}
  MaterialExpressionScalarParameter_5 {class=MaterialExpressionScalarParameter, properties {defaultValue=60, group=Ripples, parameterName=Ripples_Thining}}
  MaterialExpressionScalarParameter_6 {class=MaterialExpressionScalarParameter, properties {defaultValue=70, group=Ripples, parameterName=EmissiveIntensity}}
  MaterialExpressionSine_0 {class=MaterialExpressionSine, inputs [None {from=MaterialExpressionMultiply_6}]}
  MaterialExpressionSine_1 {class=MaterialExpressionSine, inputs [None {from=MaterialExpressionMultiply_17}]}
  MaterialExpressionSine_2 {class=MaterialExpressionSine, inputs [None {from=MaterialExpressionMultiply_26}]}
  MaterialExpressionSphereMask_0 {
   class=MaterialExpressionSphereMask
   inputs [
    A {from=MaterialExpressionNamedRerouteUsage_17, output=""}
    B {from=MaterialExpressionNamedRerouteUsage_15, output=""}
    Radius {from=MaterialExpressionNamedRerouteUsage_18, output=""}
    Hardness {from="", output=""}
   ]
   properties {attenuationRadius=1, hardnessPercent=1}
  }
  MaterialExpressionSphereMask_1 {
   class=MaterialExpressionSphereMask
   inputs [
    A {from=MaterialExpressionNamedRerouteDeclaration_0, output=""}
    B {from=MaterialExpressionNamedRerouteUsage_6, output=""}
    Radius {from=MaterialExpressionNamedRerouteUsage_8, output=""}
    Hardness {from="", output=""}
   ]
   properties {attenuationRadius=1, hardnessPercent=0}
  }
  MaterialExpressionSphereMask_2 {
   class=MaterialExpressionSphereMask
   inputs [
    A {from=MaterialExpressionNamedRerouteUsage_7, output=""}
    B {from=MaterialExpressionNamedRerouteUsage_3, output=""}
    Radius {from=MaterialExpressionNamedRerouteUsage_10, output=""}
    Hardness {from="", output=""}
   ]
   properties {attenuationRadius=1, hardnessPercent=1}
  }
  MaterialExpressionSphereMask_3 {
   class=MaterialExpressionSphereMask
   inputs [
    A {from=MaterialExpressionTextureCoordinate_0, output=""}
    B {from=MaterialExpressionNamedRerouteUsage_2, output=""}
    Radius {from=MaterialExpressionNamedRerouteUsage_9, output=""}
    Hardness {from="", output=""}
   ]
   properties {attenuationRadius=1, hardnessPercent=0}
  }
  MaterialExpressionSphereMask_5 {
   class=MaterialExpressionSphereMask
   inputs [
    A {from=MaterialExpressionNamedRerouteUsage_0, output=""}
    B {from=MaterialExpressionNamedRerouteUsage_4, output=""}
    Radius {from=MaterialExpressionConstant_17, output=""}
    Hardness {from=MaterialExpressionConstant_48, output=""}
   ]
   properties {}
  }
  MaterialExpressionSphereMask_6 {
   class=MaterialExpressionSphereMask
   inputs [
    A {from=MaterialExpressionNamedRerouteUsage_1, output=""}
    B {from=MaterialExpressionNamedRerouteUsage_5, output=""}
    Radius {from=MaterialExpressionConstant_23, output=""}
    Hardness {from=MaterialExpressionConstant_24, output=""}
   ]
   properties {}
  }
  MaterialExpressionSphereMask_7 {
   class=MaterialExpressionSphereMask
   inputs [
    A {from=MaterialExpressionNamedRerouteUsage_21, output=""}
    B {from=MaterialExpressionNamedRerouteUsage_22, output=""}
    Radius {from=MaterialExpressionConstant_12, output=""}
    Hardness {from=MaterialExpressionConstant_13, output=""}
   ]
   properties {}
  }
  MaterialExpressionSphereMask_8 {
   class=MaterialExpressionSphereMask
   inputs [
    A {from=MaterialExpressionNamedRerouteUsage_23, output=""}
    B {from=MaterialExpressionNamedRerouteUsage_24, output=""}
    Radius {from=MaterialExpressionConstant_10, output=""}
    Hardness {from=MaterialExpressionConstant_11, output=""}
   ]
   properties {}
  }
  MaterialExpressionSphereMask_9 {
   class=MaterialExpressionSphereMask
   inputs [
    A {from=MaterialExpressionNamedRerouteDeclaration_6, output=""}
    B {from=MaterialExpressionNamedRerouteUsage_30, output=""}
    Radius {from=MaterialExpressionNamedRerouteUsage_31, output=""}
    Hardness {from="", output=""}
   ]
   properties {attenuationRadius=1, hardnessPercent=0}
  }
  MaterialExpressionSubtract_0 {
   class=MaterialExpressionSubtract
   inputs [
    A {from=MaterialExpressionLinearInterpolate_4}
    B {from=""}
   ]
  }
  MaterialExpressionSubtract_1 {
   class=MaterialExpressionSubtract
   inputs [
    A {from=MaterialExpressionMultiply_8}
    B {from=MaterialExpressionSubtract_0}
   ]
  }
  MaterialExpressionSubtract_2 {
   class=MaterialExpressionSubtract
   inputs [
    A {from=MaterialExpressionAdd_3}
    B {from=MaterialExpressionMultiply_10}
   ]
  }
  MaterialExpressionSubtract_3 {
   class=MaterialExpressionSubtract
   inputs [
    A {from=MaterialExpressionTransformPosition_0}
    B {from=MaterialExpressionWorldPosition_0, output=XYZ}
   ]
  }
  MaterialExpressionSubtract_4 {
   class=MaterialExpressionSubtract
   inputs [
    A {from=MaterialExpressionConstant_32}
    B {from=""}
   ]
  }
  MaterialExpressionSubtract_5 {
   class=MaterialExpressionSubtract
   inputs [
    A {from=MaterialExpressionMultiply_18}
    B {from=MaterialExpressionSubtract_4}
   ]
  }
  MaterialExpressionSubtract_6 {
   class=MaterialExpressionSubtract
   inputs [
    A {from=MaterialExpressionNamedRerouteUsage_34}
    B {from=MaterialExpressionSphereMask_7}
   ]
  }
  MaterialExpressionSubtract_7 {
   class=MaterialExpressionSubtract
   inputs [
    A {from=MaterialExpressionConstant_38}
    B {from=""}
   ]
  }
  MaterialExpressionSubtract_8 {
   class=MaterialExpressionSubtract
   inputs [
    A {from=MaterialExpressionMultiply_32}
    B {from=MaterialExpressionSubtract_7}
   ]
  }
  MaterialExpressionSubtract_9 {
   class=MaterialExpressionSubtract
   inputs [
    A {from=MaterialExpressionAdd_11}
    B {from=MaterialExpressionMultiply_35}
   ]
  }
  MaterialExpressionTextureCoordinate_0 {class=MaterialExpressionTextureCoordinate}
  MaterialExpressionTextureCoordinate_1 {class=MaterialExpressionTextureCoordinate}
  MaterialExpressionTextureCoordinate_2 {class=MaterialExpressionTextureCoordinate}
  MaterialExpressionTextureCoordinate_3 {class=MaterialExpressionTextureCoordinate}
  MaterialExpressionTextureCoordinate_4 {class=MaterialExpressionTextureCoordinate}
  MaterialExpressionTextureCoordinate_5 {class=MaterialExpressionTextureCoordinate}
  MaterialExpressionTextureObject_0 {
   class=MaterialExpressionTextureObject
   inputs []
   properties {samplerType=SAMPLERTYPE_Grayscale, texture=/Game/Global/Textures/Default/Noise/LowResBlurredNoiseNormalized}
  }
  MaterialExpressionTextureSample_0 {
   class=MaterialExpressionTextureSample
   inputs [
    UVs {from=MaterialExpressionPanner_1}
    Tex {from=""}
    Apply View MipBias {from=""}
   ]
   properties {samplerSource=SSM_Wrap_WorldGroupSettings, samplerType=SAMPLERTYPE_Masks, texture=/Game/Global/Textures/Default/Noise/WindNoise_Mask_1}
  }
  MaterialExpressionTextureSample_1 {
   class=MaterialExpressionTextureSample
   inputs [
    UVs {from=MaterialExpressionPanner_0}
    Tex {from=""}
    Apply View MipBias {from=""}
   ]
   properties {samplerType=SAMPLERTYPE_Grayscale, texture=&LowResBlurredNoiseNormalized}
  }
  MaterialExpressionTextureSample_2 {
   class=MaterialExpressionTextureSample
   inputs [
    UVs {from=MaterialExpressionComponentMask_2}
    Tex {from=""}
    Apply View MipBias {from=""}
   ]
   properties {texture=/CosmeticCompanions/Assets/Props/Emotes/GrainSpike/Texture/T_Emote_GrainSpike_Portal_D}
  }
  MaterialExpressionTextureSample_3 {
   class=MaterialExpressionTextureSample
   inputs [
    UVs {from=MaterialExpressionPanner_3}
    Tex {from=""}
    Apply View MipBias {from=""}
   ]
   properties {samplerSource=SSM_Wrap_WorldGroupSettings, texture=/Game/Global/Textures/Default/PointSplatter/BoatFlake}
  }
  MaterialExpressionTextureSample_4 {
   class=MaterialExpressionTextureSample
   inputs [
    UVs {from=MaterialExpressionMultiply_23}
    Tex {from=MaterialExpressionTextureObject_0}
    "DDX(UVs)" {from=MaterialExpressionDDX_0}
    "DDY(UVs)" {from=MaterialExpressionDDY_0}
    Apply View MipBias {from=""}
   ]
   properties {mipValueMode=TMVM_Derivative, texture=&T_Emote_GrainSpike_Portal_D}
  }
  MaterialExpressionTime_0 {class=MaterialExpressionTime}
  MaterialExpressionTime_1 {class=MaterialExpressionTime}
  MaterialExpressionTime_2 {class=MaterialExpressionTime}
  MaterialExpressionTransformPosition_0 {
   class=MaterialExpressionTransformPosition
   inputs [None {from=MaterialExpressionConstant3Vector_0, output=""}]
   properties {transformType=TRANSFORMPOSSOURCE_World}
  }
  MaterialExpressionVectorParameter_0 {class=MaterialExpressionVectorParameter, properties {group=Ripples, parameterName=RippleColor, defaultValue=<rgba 0.0176 0.2936 0.88 1>}}
  MaterialExpressionWorldPosition_0 {class=MaterialExpressionWorldPosition, inputs [], properties {}}
 }
 properties {
  bAutomaticallySetUsageInEditor=false
  bUsedWithMeshDeformer=true
  bUsedWithSkeletalMesh=true
  blendMode=BLEND_Translucent
  twoSided=true
  usedShadingModels=Default Lit
 }
}
